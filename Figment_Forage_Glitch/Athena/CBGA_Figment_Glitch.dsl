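; Glitch forageable consumable that periodically changes into a random item type, each with its own effect and consume rules.
; A server timer rolls the type, using scalable-float hotfix values to replace disabled picks with a default, and selects the spooky-mist or generic consume ability.
; OnRep_Type maps the type to gameplay effect, montage, mesh and fail reason, then triggers swap effects, looping audio and scale animation until interaction.
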(event Interaction|EventBlueprintOnInteract
  (params (InteractingPawn :type Fort%20Pawn%20Object%20Reference) (InteractionBeingAttempted :type EInteractionBeingAttempted%20Enum :default "FirstInteraction"))
  ;[1 1 1] Check if we have interacted
  (Variables|Default|SetInteracted true)
  (Development|PrintString "InteractHappened"))

(event EventBeginPlay
  ;[1 1 1] Give the mesh a random rot when it starts that makes them feel more randomly placed.
  (if (Variables|Default|GetbShouldRandomizeRotation)
    (Transformation|SetWorldRotation (Variables|Default|GetRepositioned) (Math|Rotator|MakeRotator 0.0 0.0 (Math|Random|RandomFloatinRange 0.0 3.6e+02)) false false)
    (Transformation|SetRelativeLocation (Variables|Default|GetRepositioned) (Variables|Default|GetHeightOffset) false false)
    (else
      _))
  ;[1 1 1] On a timer swap the consumable
  (if (Networking|IsServer)
    (Class|FortAsyncActionCurrentPlaylistReady|PlaylistDataReadyAsync
      (:OnReady
        (bind _returnvalue (Utilities|Time|SetTimerbyEvent (delegate Custom|GlitchSwap) 5.0 true false 0.0 (Math|Random|RandomFloatinRange 1.0 5.0)))
        (Variables|Default|SetTimer_Change _returnvalue)))))

(event Custom|GlitchSwap
  ;[1 1 1] On a timer swap the consumable
  (bind _type (CallFunction|CheckifEnabled (Utilities|Enum|BytetoEnumEnum_Glitch_ForagedItems (Math|Conversions|ToByte(Integer) (Math|Random|RandomIntegerinRange 0 5)))))
  (bind _output_get (Variables|Default|SetType _type))
  ;[1 1 1] Set consume ability based off if spooky mist or not
  (Variables|Default|SetConsumeAbilityToUse (select (Utilities|Enum|Equal(Enum) _output_get "NewEnumerator3") (Variables|Default|GetSpookyMistConsume) (Variables|Default|GetGenericConsume))))

(event Custom|SwapVisuals
  ;[1 1 1] Play Visuals and sounds
  (if (not (Variables|Default|GetInteracted))
    (Effects|SpawnEmitteratLocation (Variables|Default|GetSwapParticle) (Transformation|GetActorLocation) (Transformation|GetActorRotation) (Variables|Default|GetFXScale))
    (Audio|PlaySoundatLocation (Variables|Audio|GetSwapSound) (Transformation|GetActorLocation))
    (CallFunction|PlayAudioLoop)
    (bind (_direction _scale) (Timeline|Scale PlayFromStart)
      (:Update
        (Transformation|SetRelativeScale3D (Variables|Default|GetRepositioned) (Math|Conversions|ToVector(Float) _scale))))))

(event EventOnDeathPlayEffects
  (params (Damage :type Float%20%28single-precision%29) (DamageTags :type Gameplay%20Tag%20Container%20Structure%20%28by%20ref%29) (Momentum :type Vector :default "0, 0, 0") (HitInfo :type Hit%20Result%20Structure%20%28by%20ref%29) (InstigatedBy :type Fort%20Pawn%20Object%20Reference) (DamageCauser :type Actor%20Object%20Reference) (EffectContext :type Gameplay%20Effect%20Context%20Handle%20Structure))
  ;[1 1 1] Play Visuals and sounds
  (bind _foarged_item_audio_loop (Variables|Default|GetFoarged%20Item%20Audio%20Loop)
    (:then
      (Audio|Components|Audio|FadeOut _foarged_item_audio_loop 0.25 0.0 "Linear" :!class /Script/Engine.AudioComponent))
    (:else)))

(fn BlueprintCanInteract
  (params (InteractingPawn :type Fort%20Pawn%20Object%20Reference) (InteractionBeingAttempted :type EInteractionBeingAttempted%20Enum :default "FirstInteraction") (InteractionType :type TInteractionType%20Enum :default "IT_NoInteraction"))
  (return true))

(fn CheckIfEnabled
  (params (Key :type Enum_Glitch_ForagedItems%20Enum%20%28by%20ref%29 :default "NewEnumerator0"))
  (bind _map_typetoenabled (Variables|Default|GetMAP_TypeToEnabled))
  (bind (_value _returnvalue) (Utilities|Map|Find _map_typetoenabled Key))
  ;[1 1 1] Check hotfix to see if the item type is enabled
  (if _returnvalue
    (return (select (Math|Conversions|ToBoolean(Integer) (Math|Float|Ceil (ScalableFloat|GetValueatLevel _value))) Key (CallFunction|GetRerollType)))
    (else
      (return (CallFunction|GetRerollType)))))

(fn DetermineOnConsumeGameplayEffect
  (return (Variables|Default|GetGE_ToApply)))

(fn DetermineOnConsumeMontage
  (return (Variables|Default|GetMontageToUse)))

(fn DetermineOnConsumeMontageAbility
  (return (Variables|Default|GetConsumeAbilityToUse)))

(fn GetRerollType
  ;[1 1 1] Get the value of the sheet and convert it to the consumable enum
  (return (Utilities|Enum|BytetoEnumEnum_Glitch_ForagedItems (Math|Conversions|ToByte(Integer) (Math|Float|Ceil (ScalableFloat|GetValueatLevel (Variables|Default|GetRow_RerollDefault)))))))

(fn OnInteractAllowedToConsume
  (params (InteractingPawn :type Fort%20Pawn%20Object%20Reference) (InteractionBeingAttempted :type EInteractionBeingAttempted%20Enum :default "FirstInteraction"))
  (bind _returnvalue (FortPawn|GetShield InteractingPawn))
  (bind _returnvalue_2 (FortPawn|GetHealth InteractingPawn))
  (bind _returnvalue_3 (FortPawn|GetMaxHealth InteractingPawn))
  (bind _returnvalue_4 (Math|Float|NearlyEqual(Float) _returnvalue_2 _returnvalue_3))
  (bind _returnvalue_5 (not _returnvalue_4))
  (bind _returnvalue_6 (FortPawn|GetMaxShield InteractingPawn))
  (bind _returnvalue_7 (Math|Float|NearlyEqual(Float) _returnvalue _returnvalue_6))
  (bind _returnvalue_8 (not _returnvalue_7))
  (Utilities|IsValid InteractingPawn
    (:Is%20Valid
      (switch Utilities|FlowControl|Switch|SwitchonEnum_Glitch_ForagedItems (Variables|Default|GetType)
        (:NewEnumerator0
          (return _returnvalue_5))
        (:NewEnumerator1
          (return (or _returnvalue_8 _returnvalue_5)))
        (:NewEnumerator2
          _)
        (:NewEnumerator3
          (return (not (GameplayTags|TagContainer|CompareTagAssetInterfacetoOtherTags (LocalVariables|GetSpookyTag) InteractingPawn))))
        (:NewEnumerator4
          _)
        (:NewEnumerator5
          (return _returnvalue_8)))
      (return true))
    (:Is%20Not%20Valid
      (return false))))

(fn OnRep_Type
  (bind _type (Variables|Default|GetType))
  (bind _map_effecttotype (Variables|Default|GetMAP_EffectToType))
  (bind (_value _returnvalue) (Utilities|Map|Find _map_effecttotype _type))
  (bind _map_montagetotype (Variables|Default|GetMAP_MontageToType))
  (bind (_value_2 _returnvalue_2) (Utilities|Map|Find _map_montagetotype _type))
  (bind _map_meshtotype (Variables|Default|GetMAP_MeshToType))
  (bind (_value_3 _returnvalue_3) (Utilities|Map|Find _map_meshtotype _type))
  (bind _map_failedreason (Variables|Default|GetMAP_FailedReason))
  (bind (_value_4 _returnvalue_4) (Utilities|Map|Find _map_failedreason _type))
  (bind _type_2 (Variables|Default|GetType))
  ;[1 1 1] Set the GE, Montage, and GEs to use based on the type of consumable. Turn on other visuals if needed.
  (if _returnvalue
    (Variables|Default|SetGE_ToApply _value)
    (if _returnvalue_3
      (Components|StaticMesh|SetStaticMesh (Variables|Default|GetRepositioned) _value_3)
      (if _returnvalue_2
        (Variables|Default|SetMontageToUse _value_2)
        (if _returnvalue_4
          (Variables|FailedReason|SetFailedReason _value_4)
          (CallFunction|SwapVisuals)
          (if (Utilities|Enum|Equal(Enum) _type_2 "NewEnumerator3")
            (Components|Activation|Activate (Variables|Default|GetNS_MistFX) false)
            ;[1 1 1] Extra modification, deactivate the rock FX when they replace each other.
            (Components|Activation|Deactivate (Variables|Default|GetGravFX))
            (elif (Utilities|Enum|Equal(Enum) _type_2 "NewEnumerator4")
              (Components|Activation|Activate (Variables|Default|GetGravFX) false)
              (Components|Activation|Deactivate (Variables|Default|GetNS_MistFX))
              (else
                (Components|Activation|Deactivate (Variables|Default|GetGravFX))))))))))

(fn PlayAudioLoop
  (bind _type (Variables|Default|GetType))
  (bind _map_soundtotype (Variables|Audio|GetMAP_SoundToType))
  (bind (_value _returnvalue) (Utilities|Map|Find _map_soundtotype _type))
  (if _returnvalue
    (bind _foarged_item_audio_loop (Variables|Default|GetFoarged%20Item%20Audio%20Loop)
      (:then
        (Audio|Components|Audio|FadeOut _foarged_item_audio_loop 0.35 0.0 "Linear" :!class /Script/Engine.AudioComponent))
      (:else
        _))
    (bind _returnvalue_2 (Audio|SpawnSoundAttached _value (Variables|Default|GetStaticMesh)))
    (bind _output_get (Variables|Default|SetFoarged%20Item%20Audio%20Loop _returnvalue_2))
    (Audio|Components|Audio|FadeIn _output_get 0.35 1.0 0.0 "Linear" :!class /Script/Engine.AudioComponent)
    (else
      (bind _foarged_item_audio_loop_2 (Variables|Default|GetFoarged%20Item%20Audio%20Loop)
        (:then
          (Audio|Components|Audio|FadeOut _foarged_item_audio_loop_2 0.5 0.0 "Linear" :!class /Script/Engine.AudioComponent))
        (:else)))))

(fn ConstructionScript
  (|CallParentFunction))
